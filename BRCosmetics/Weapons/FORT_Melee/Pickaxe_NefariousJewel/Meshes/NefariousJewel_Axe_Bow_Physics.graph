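# Physics asset for the bow variant of the NefariousJewel two-handed pickaxe, defining bodies and joints for its dangling bow and crymbal parts.
# Capsule bodies cover the handle, hinges, bow and crymbal bones.
# Angularly locked constraints tie the hinges and base bones back to the handle.
# Unlocked joints carry the dyn_bow pieces, two branching from dyn_bow_01, and dyn_crymbal_02, whose configured limit angles are far tighter than the bow's.
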PhysicsAsset /BRCosmetics/Weapons/FORT_Melee/Pickaxe_NefariousJewel/Meshes/NefariousJewel_Axe_Bow_Physics
bodies {
 Melee_TwoHanded_Handle [{shapeType=Capsule, length=45.7521, radius=3.6528, center=<0.8038 -0.8302 8.4137>, rotation=<rot 0 -90 0>}]
 base_bow_01 [{shapeType=Capsule, length=0.505, radius=0.505, shapeName=base_bow_01_capsule, rotation=<rot 0 -90 0>}]
 base_crymbal_02 [{shapeType=Capsule, length=0.505, radius=0.505, shapeName=base_crymbal_02_capsule, rotation=<rot 0 -90 0>}]
 base_hinge__02 [{shapeType=Capsule, length=4, radius=5, shapeName=base_hinge__02_capsule, rotation=<rot 0 -90 0>}]
 dyn_bow_01 [{shapeType=Capsule, length=2, radius=2, shapeName=dyn_bow_01_capsule, center=<0.8989 0 0.1273>, rotation=<rot 12 0 0>}]
 dyn_bow_02 [{shapeType=Capsule, length=4, radius=1, shapeName=dyn_bow_02_capsule, center=<3.8613 0 0.2532>, rotation=<rot 0 -90 -84.4504>}]
 dyn_bow_03 [{shapeType=Capsule, length=4, radius=1, shapeName=dyn_bow_03_capsule, center=<2.6477 0 0.0631>, rotation=<rot 0 90 -92.7836>}]
 dyn_crymbal_02 [{shapeType=Capsule, length=20, radius=14, shapeName=dyn_crymbal_02_capsule, center=<17.9683 0 9.6378>, rotation=<rot -85 180 -180>}]
 dyn_hinge_01 [
  {shapeType=Capsule, length=21, radius=6, center=<6.597 8.0463 0.0149>, rotation=<rot -0.0007 83.0361 -89.8461>}
  {shapeType=Capsule, length=2.6856, radius=5.5826, shapeName=dyn_hinge_01_capsule, center=<-0.1208 -0.2754 0>}
 ]
}
constraints [
 {
  bone1Name=dyn_hinge_01
  bone2Name=Melee_TwoHanded_Handle
  swing1LimitDegrees=45
  swing1Motion=Locked
  swing2LimitDegrees=45
  swing2Motion=Locked
  twistLimitDegrees=45
  twistMotion=Locked
 }
 {
  bone1Name=base_bow_01
  bone2Name=dyn_hinge_01
  swing1LimitDegrees=45
  swing1Motion=Locked
  swing2LimitDegrees=45
  swing2Motion=Locked
  twistLimitDegrees=45
  twistMotion=Locked
 }
 {bone1Name=dyn_bow_01, bone2Name=base_bow_01, swing1LimitDegrees=60, swing2LimitDegrees=40, twistLimitDegrees=2}
 {bone1Name=dyn_bow_02, bone2Name=dyn_bow_01, swing1LimitDegrees=60, swing2LimitDegrees=50, twistLimitDegrees=5}
 {bone1Name=dyn_bow_03, bone2Name=dyn_bow_01, swing1LimitDegrees=60, swing2LimitDegrees=50, twistLimitDegrees=5}
 {
  bone1Name=base_crymbal_02
  bone2Name=base_hinge__02
  swing1LimitDegrees=45
  swing1Motion=Locked
  swing2LimitDegrees=45
  swing2Motion=Locked
  twistLimitDegrees=45
  twistMotion=Locked
 }
 {bone1Name=dyn_crymbal_02, bone2Name=base_crymbal_02, swing1LimitDegrees=5, swing2LimitDegrees=0, twistLimitDegrees=0}
 {
  bone1Name=base_hinge__02
  bone2Name=dyn_hinge_01
  swing1LimitDegrees=45
  swing1Motion=Locked
  swing2LimitDegrees=45
  swing2Motion=Locked
  twistLimitDegrees=45
  twistMotion=Locked
 }
]
